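# Post-process AnimBP for the TwiceBaked character head that limits head bone rotation when the hood-up style is active.
# Each update reads the Stage2 cosmetic variant and the head's delta from ref pose relative to neck_02.
# That delta is negated and scaled per axis, then applied as additive bone-space rotation, blended by the hood flag.
# Hooded scaling counters 60% of one axis, unhooded scaling is zero, and alpha requires a valid pawn owner.
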AnimBlueprint /BRCosmetics/Characters/Player/Male/Medium/Heads/M_MED_TwiceBaked_Head/Meshes/M_MED_TwiceBaked_PostProcess_Head_AnimBP
target_skeleton=/Game/Characters/Player/Common/Fortnite_Base_Head/Fortnite_Base_Head_Export_Skeleton
graphs [
 AnimGraph {nodes [
   n1 = "Transform (Modify) Bone - Bone: head"(
    ComponentPose=Local To Component(LocalPose=Use cached pose 'inputPoseCache'())
    Alpha=Get enableHeadRotLimitAlpha() -> (enableHeadRotLimitAlpha:double)
    Translation="0,0,0"
    Rotation=Get headRotIsHoodedVariant() -> (headRotIsHoodedVariant:Rotator)
    Scale="1,1,1"
   ) {
    rotationMode=BMM_Additive
    rotationSpace=BCS_BoneSpace
    alphaBoolBlend {blendOption=Linear}
    boneToModify {boneName=head}
    rotation=<rot 50 50 50>
   }
   n2 = "Transform (Modify) Bone - Bone: head"(
    ComponentPose=Local To Component(LocalPose=Use cached pose 'inputPoseCache'())
    Alpha=Get enableHeadRotLimitAlpha()
    Translation="0,0,0"
    Rotation=Get headRotIsNotHoodedVariant() -> (headRotIsNotHoodedVariant:Rotator)
    Scale="1,1,1"
   ) {rotationMode=BMM_Additive, rotationSpace=BCS_BoneSpace, alphaBoolBlend {blendOption=Linear}, boneToModify {boneName=head}}
   Output Pose(Result=Blend Poses by bool(
     BlendPose_0=Component To Local(ComponentPose=@n1)
     BlendPose_1=Component To Local(ComponentPose=@n2)
     bActiveValue=Get hoodOnVariant() -> (hoodOnVariant:bool)
    ) {blendType=Linear})
   Save cached pose 'inputPoseCache'(Pose=Input Pose()) {cacheName=inputPoseCache}
  ]}
 EventGraph {nodes [
   Event BlueprintUpdateAnimation()
   > n1 = Sequence!(execute) -> (then_0, then_1, then_2:Exec, then_3:Exec)
   > n2 = Cast To FortPlayerPawn(Object=GetOwningActor())
   > n3 = GetVariantChannelActiveIndexForItemActor!(
    execute
    Item:Athena Cosmetic Item Definition Object Reference=/BRCosmetics/Athena/Items/Cosmetics/Characters/Character_TwiceBaked
    Actor:Actor Object Reference=@n2
    Channel:Gameplay Tag Structure=(TagName="Cosmetics.Variant.Channel.Parts")
    VariantList:Array of Gameplay Tag Structures=Make Array("[0]"=(TagName="Cosmetics.Variant.Property.Stage2"))
    bErrorIfChannelMissing:bool=false
    bErrorIfTagNotInArray:bool=false
   ) -> (then, Index:int=0, bIndexIsValid:bool, bFoundChannel:bool)
   > n4 = Branch(Condition=@n3.bIndexIsValid)
   > Set hoodOnVariant(hoodOnVariant:bool=true) -> (Output_Get:bool)
   Set hoodOnVariant(execute=@n4.else, hoodOnVariant=false)
   n5 = Sequence!(execute=@n1.then_1) -> (then_0, then_1, then_2, then_3, then_4:Exec, then_5:Exec, then_6:Exec, then_7:Exec)
   > n6 = Set headRotScaleIsEmoting(headRotScaleIsEmoting:Rotator="0,0,0.6") -> (Output_Get:Rotator)
   n7 = Set headRotScaleIsNotEmoting(execute=@n5.then_1, headRotScaleIsNotEmoting:Rotator="0, 0, 0") -> (Output_Get:Rotator)
   n8 = BreakRotator(InRot=@n6)
   Set X percentageEmote(execute=@n5.then_2, X percentageEmote:float=@n8) -> (Output_Get:float)
   Set Y percentageEmote(execute=@n5.then_3, Y percentageEmote:float=@n8.Pitch) -> (Output_Get:float)
   Set Z percentageEmote(execute=@n5.then_4, Z percentageEmote:float=@n8.Yaw) -> (Output_Get:float)
   n9 = BreakRotator(InRot=@n7)
   Set X percentageNotEmote(execute=@n5.then_5, X percentageNotEmote:float=@n9) -> (Output_Get:float)
   Set Y percentageNotEmote(execute=@n5.then_6, Y percentageNotEmote:float=@n9.Pitch) -> (Output_Get:float)
   Set Z percentageNotEmote(execute=@n5.then_7, Z percentageNotEmote:float=@n9.Yaw) -> (Output_Get:float)
   n10 = Is Valid(exec=@n1.then_2, InputObject=TryGetPawnOwner())
   Set enableHeadRotLimitAlpha(execute=@n10.Is Not Valid, enableHeadRotLimitAlpha:double=0) -> (Output_Get:double)
   Set enableHeadRotLimitAlpha(execute=@n10.Is Valid, enableHeadRotLimitAlpha=1)
   n11 = BreakTransform(InTransform=GetDeltaTransformFromRefPose(self=GetOwnerSkeletalMeshComponent(), BoneName=head, BaseName=neck_02))
   n12 = BreakRotator(InRot=@n11.Rotation)
   n13 = Get Z percentageEmote() -> (Z percentageEmote:float)
   n14 = Get X percentageEmote() -> (X percentageEmote:float)
   n15 = Get Y percentageEmote() -> (Y percentageEmote:float)
   > Set headRotIsHoodedVariant(headRotIsHoodedVariant:Rotator=MakeRotator(
     Roll=float * float!(A=@n12, B=-1, C:double=@n14) -> (ReturnValue)
     Pitch=float * float(A=@n12.Pitch, B=-1, C=@n15)
     Yaw=float * float(A=@n12.Yaw, B=-1, C=@n13)
    )) -> (Output_Get:Rotator)
   n16 = BreakRotator(InRot=@n11.Rotation)
   n17 = Get Y percentageNotEmote() -> (Y percentageNotEmote:float)
   n18 = Get X percentageNotEmote() -> (X percentageNotEmote:float)
   n19 = Get Z percentageNotEmote() -> (Z percentageNotEmote:float)
   > Set headRotIsNotHoodedVariant(headRotIsNotHoodedVariant:Rotator=MakeRotator(Roll=float * float(A=@n16, B=-1, C=@n18), Pitch=float * float(A=@n16.Pitch, B=-1, C=@n17), Yaw=float * float(A=@n16.Yaw, B=-1, C=@n19))) -> (Output_Get:Rotator)
  ]}
]
variables [
 enableHeadRotLimitAlpha {type=number}
 headRotIsHoodedVariant {type=Rotator}
 headRotIsNotHoodedVariant {type=Rotator}
 headRotScaleIsEmoting {type=Rotator}
 headRotScaleIsNotEmoting {type=Rotator}
 hoodOnVariant {type=boolean}
 X percentageEmote {type=number}
 X percentageNotEmote {type=number}
 Y percentageEmote {type=number}
 Y percentageNotEmote {type=number}
 Z percentageEmote {type=number}
 Z percentageNotEmote {type=number}
]
